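; Server-only inventory device that applies the BR-style itemization configuration to the minigame.
; On BeginPlay and again on minigame start, it gets the minigame settings machine, with a half-second delay branch if that is invalid.
; It then sets the custom inventory configuration on that machine and prints an activation message.
; Minigame end has no logic, and Play, Pause and Stop only return false.
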(event EventBeginPlay
  (bind _self self)
  (bind _returnvalue (Minigame|GetMinigameSettingsMachine _self))
  (Utilities|IsValid _returnvalue
    (:Is%20Valid
      _)
    (:Is%20Not%20Valid
      (Utilities|FlowControl|Delay 0.5)))
  (Class|FortMinigameSettingsBuilding|SetCustomInventoryConfiguration _returnvalue "/Itemization/BRStyle/ItemizationConfiguration_BRStyle.ItemizationConfiguration_BRStyle")
  (Development|PrintString "Inventory Device activated"))

(event EventOnMinigameStarted
  (bind _self self)
  (bind _returnvalue (Minigame|GetMinigameSettingsMachine _self))
  (Utilities|IsValid _returnvalue
    (:Is%20Valid
      _)
    (:Is%20Not%20Valid
      (Utilities|FlowControl|Delay 0.5)))
  (Class|FortMinigameSettingsBuilding|SetCustomInventoryConfiguration _returnvalue "/Itemization/BRStyle/ItemizationConfiguration_BRStyle.ItemizationConfiguration_BRStyle")
  (Development|PrintString "Inventory Device activated"))

(event EventOnMinigameEnded)

(fn GetExecutionContext
  (return "ServerOnly"))

(fn Pause
  (return false))

(fn Play
  (return false))

(fn Stop
  (return false))

(fn ConstructionScript
  (|CallParentFunction))
